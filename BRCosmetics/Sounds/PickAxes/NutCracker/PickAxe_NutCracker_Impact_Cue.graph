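# Sound cue for the impact hit of the NutCracker cosmetic pickaxe.
# A random node picks evenly among five impact recordings, and a modulator above it varies pitch and volume on each play.
# It routes to the impacts sound class, uses shared pickaxe attenuation, property sheet and player melee concurrency assets, and primes on load.
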SoundCue /BRCosmetics/Sounds/PickAxes/NutCracker/PickAxe_NutCracker_Impact_Cue
root=Modulator_0
nodes [
 {id=Modulator_0, children [Random_0], properties {pitchMax=1.2, pitchMin=0.9, volumeMin=0.75}}
 {id=Random_0, children [WavePlayer_0,WavePlayer_1,WavePlayer_2,WavePlayer_5,WavePlayer_8], properties {weights [1,1,1,1,1]}}
 {id=WavePlayer_0, properties {soundWaveAssetPtr=&/Pickaxe_Nutcracker_Impact_01}}
 {id=WavePlayer_1, properties {soundWaveAssetPtr=&/Pickaxe_Nutcracker_Impact_02}}
 {id=WavePlayer_2, properties {soundWaveAssetPtr=&/Pickaxe_Nutcracker_Impact_03}}
 {id=WavePlayer_5, properties {soundWaveAssetPtr=&/Pickaxe_Nutcracker_Impact_04}}
 {id=WavePlayer_8, properties {soundWaveAssetPtr=&/Pickaxe_Nutcracker_Impact_05}}
]
properties {
 attenuationSettings=/Game/Sounds/Attenuation/Fort_Pickaxe_Impact_Att
 audioPropertiesSheet=/BRCosmetics/Sounds/Settings/PropertySheets/Pickaxe/APS_Cosmetic_Pickaxe_Impact
 bPrimeOnLoad=true
 maxDistance=4300
 pitchMultiplier=1.05
 soundClassObject=&Fort_Impacts
 volumeMultiplier=0.825
 attenuationOverrides {
  bAttenuateWithLPF=true
  distanceAlgorithm=Logarithmic
  falloffDistance=800
  lPFRadiusMax=800
  lPFRadiusMin=400
  attenuationShapeExtents {x=300}
 }
 concurrencyOverrides {bLimitToOwner=true, maxCount=2, resolutionRule=StopOldest}
 concurrencySet [&Player_Melee_Weapon_Conc]
}
